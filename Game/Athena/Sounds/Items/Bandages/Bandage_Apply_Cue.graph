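# Sound cue for applying a bandage consumable in Athena, playing the single Bandage_Use wave through one wave player.
# It routes to the Fort_Consumables sound class at reduced volume and uses the shared Athena consumable-use attenuation.
# A concurrency override caps simultaneous instances and stops the oldest when the limit is reached.
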SoundCue /Game/Athena/Sounds/Items/Bandages/Bandage_Apply_Cue
root=WavePlayer_3
nodes [{id=WavePlayer_3, properties {soundWaveAssetPtr=&/Bandage_Use}}]
properties {
 attenuationSettings=&Fort_Consumable_Use_Athena_Att
 bOverrideConcurrency=true
 maxDistance=2250
 soundClassObject=&Fort_Consumables
 volumeMultiplier=0.6
 concurrencyOverrides {maxCount=4, resolutionRule=StopOldest}
}
