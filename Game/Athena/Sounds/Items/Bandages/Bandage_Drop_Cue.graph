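# Sound cue for dropping a bandage item in Athena, played at reduced volume.
# A Random node picks one of four equally weighted bandage drop waves and feeds a Modulator that slightly varies pitch for variety.
# It uses the shared weapon drop attenuation and the loot collect sound class.
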SoundCue /Game/Athena/Sounds/Items/Bandages/Bandage_Drop_Cue
root=Modulator_1
nodes [
 {id=Modulator_1, children [Random_1], properties {pitchMax=1.1, pitchMin=0.9}}
 {id=Random_1, children [WavePlayer_0,WavePlayer_5,WavePlayer_6,WavePlayer_7], properties {weights [1,1,1,1]}}
 {id=WavePlayer_0, properties {soundWaveAssetPtr=&/Bandage_Drop_01}}
 {id=WavePlayer_5, properties {soundWaveAssetPtr=&/Bandage_Drop_02}}
 {id=WavePlayer_6, properties {soundWaveAssetPtr=&/Bandage_Drop_03}}
 {id=WavePlayer_7, properties {soundWaveAssetPtr=&/Bandage_Drop_04}}
]
properties {
 attenuationSettings=&Fort_Weapon_Drop_Att
 maxDistance=3000
 soundClassObject=&Fort_Loot_Collect
 volumeMultiplier=0.5
}
