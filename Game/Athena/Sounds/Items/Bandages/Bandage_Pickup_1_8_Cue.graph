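# Sound cue for a bandage item pickup in Athena, routed to the gadgets and items sound class with weapon pickup attenuation.
# A mixer layers an immediate whoosh with two delayed UI sounds, a slightly quieter big hit at a fixed delay and a dress-up sound at a randomized delay.
# A root modulator randomly shifts the combined layers upward in pitch.
# A concurrency override prevents new instances once the limit is reached.
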SoundCue /Game/Athena/Sounds/Items/Bandages/Bandage_Pickup_1_8_Cue
root=Modulator_1
nodes [
 {id=Delay_0, children [WavePlayer_4], properties {delayMax=0.15, delayMin=0.1}}
 {id=Delay_1, children [WavePlayer_3], properties {delayMax=0.12, delayMin=0.12}}
 {id=Mixer_0, children [Delay_1,Delay_0,WavePlayer_2], properties {inputVolume [0.7,1,1]}}
 {id=Modulator_1, children [Mixer_0], properties {pitchMax=1.3, pitchMin=1.2}}
 {id=WavePlayer_2, properties {soundWaveAssetPtr=/Game/Sounds/Fort_Foley_Sounds/Whooshes/player_swing_whoosh_stereo_03}}
 {id=WavePlayer_3, properties {soundWaveAssetPtr=/Game/Sounds/Fort_UI_Sounds/FrontEnd/fort_ui_big_hit_01}}
 {id=WavePlayer_4, properties {soundWaveAssetPtr=&/fort_ui_dressup_02}}
]
properties {
 attenuationSettings=&Fort_Weapon_Pickup_Att
 bOverrideConcurrency=true
 maxDistance=3500
 soundClassObject=&Fort_Gadgets_Items
 volumeMultiplier=0.4
 concurrencyOverrides {resolutionRule=PreventNew}
}
