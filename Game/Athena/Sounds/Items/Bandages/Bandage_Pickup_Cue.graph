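# Sound cue for picking up bandage items in Athena, classed under gadget items and played with weapon pickup attenuation.
# A Random node picks one of two equally weighted pickup waves, excluded from branch culling.
# A Modulator at the root adds slight random pitch variation to each play.
# The cue lowers its volume and overrides concurrency so new instances are prevented once the limit is reached.
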SoundCue /Game/Athena/Sounds/Items/Bandages/Bandage_Pickup_Cue
root=Modulator_0
nodes [
 {id=Modulator_0, children [Random_0], properties {pitchMax=1.1, pitchMin=0.9}}
 {id=Random_0, children [WavePlayer_0,WavePlayer_1], properties {bShouldExcludeFromBranchCulling=true, weights [1,1]}}
 {id=WavePlayer_0, properties {soundWaveAssetPtr=&/Bandage_Pickup_01}}
 {id=WavePlayer_1, properties {soundWaveAssetPtr=&/Bandage_Pickup_02}}
]
properties {
 attenuationSettings=&Fort_Weapon_Pickup_Att
 bOverrideConcurrency=true
 maxDistance=3500
 soundClassObject=&Fort_Gadgets_Items
 volumeMultiplier=0.4
 concurrencyOverrides {resolutionRule=PreventNew}
}
